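# Post-process style cosmetic AnimBlueprint for the PumpkinMan Commando body, adding secondary motion to its dangling costume bones.
# It copies the attached parent's pose, constrains the tongue to spine and clavicle, then chains Anim Dynamics for tongue tip, armband, leaves, strands, vines and boot fork.
# Spine twist, measured each update as pelvis_out, blends in a corrective pouch-leaf offset.
# Final Flap nodes drive scarf_back from head rotation.
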AnimBlueprint /BRCosmetics/Characters/Player/Male/Medium/Bodies/M_MED_Commando_PumpkinMan/Meshes/M_MED_Commando_PumpkinMan_Body_AnimBP
target_skeleton=&/M_MED_Commando_PumpkinMan_Body_Skeleton
graphs [
 AnimGraph {nodes [
   Anim Dynamics - Bone: dyn_tongue_01(Alpha=0.45, ExternalForce="0,0,0") {
    angularDampingOverride=0.985
    angularSpringConstant=200
    bAngularSpring=true
    bGravityOverrideInSimSpace=true
    bLinearSpring=true
    bOverrideAngularDamping=true
    bOverrideLinearDamping=true
    gravityScale=0.275
    linearDampingOverride=0.99
    linearSpringConstant=200
    simulationSpace=RootRelative
    alphaBoolBlend {blendOption=Linear}
    boundBone {boneName=dyn_tongue_01}
    componentAppliedLinearAccClamp=<1000 1000 1000>
    componentLinearVelScale=<100 100 100>
    physicsBodyDefinitions [{
      sphereCollisionRadius=0
      boundBone {boneName=dyn_tongue_01}
      boxExtents=<30 30 30>
      constraintSetup {angularTargetAxis=AxisY, angularLimitsMax=<0 5 20>, angularLimitsMin=<0 -15 -20>, angularTarget=<0 0 0>}
      localJointOffset=<-10 0 0>
     }]
    relativeSpaceBone {boneName=spine_05}
   }
   n1 = "Constraint - dyn_tongue_01 (PPP)"!(
    ComponentPose=Local To Component(LocalPose=Copy Pose From Mesh() {bCopyCurves=true, bUseAttachedParent=true})
    Alpha=1
    ConstraintWeights_0=0.424
    ConstraintWeights_1:float=0.2922
    ConstraintWeights_2:float=0.2838
   ) -> (Pose) {
    alphaBoolBlend {blendOption=Linear}
    boneToModify {boneName=dyn_tongue_01}
    constraintSetup [{targetBone {boneName=spine_03}}, {targetBone {boneName=spine_04}}, {targetBone {boneName=clavicle_l}}]
    constraintWeights [1,1,1]
   }
   n2 = Anim Dynamics - Bone: dyn_armband_knot(ComponentPose=@n1, Alpha=0.75, ExternalForce="0,0,0") {
    alpha=0.9
    angularDampingOverride=0.98
    angularSpringConstant=1500
    bAngularSpring=true
    bGravityOverrideInSimSpace=true
    bOverrideAngularDamping=true
    numSolverIterationsPostUpdate=4
    numSolverIterationsPreUpdate=16
    simulationSpace=Actor
    alphaBoolBlend {blendOption=Linear}
    boundBone {boneName=dyn_armband_knot}
    physicsBodyDefinitions [{
      sphereCollisionRadius=0
      boundBone {boneName=dyn_armband_knot}
      boxExtents=<15 5 5>
      constraintSetup {angularTargetAxis=AxisZ, twistAxis=AxisY, angularLimitsMax=<15 0 15>, angularLimitsMin=<-15 0 -15>, angularTarget=<0 0 1>}
      localJointOffset=<0 0 0>
     }]
   }
   n3 = Anim Dynamics - Bone: dyn_armband_low(ComponentPose=@n2, Alpha=0.5, ExternalForce="0,0,0") {
    alpha=0.9
    angularDampingOverride=0.98
    angularSpringConstant=1500
    bAngularSpring=true
    bGravityOverrideInSimSpace=true
    bOverrideAngularDamping=true
    numSolverIterationsPostUpdate=4
    numSolverIterationsPreUpdate=16
    simulationSpace=Actor
    alphaBoolBlend {blendOption=Linear}
    boundBone {boneName=dyn_armband_low}
    physicsBodyDefinitions [{
      sphereCollisionRadius=0
      boundBone {boneName=dyn_armband_low}
      boxExtents=<15 5 5>
      constraintSetup {angularTargetAxis=AxisZ, twistAxis=AxisY, angularLimitsMax=<15 15 15>, angularLimitsMin=<-15 -15 -15>, angularTarget=<0 0 1>}
      localJointOffset=<0 0 0>
     }]
   }
   n4 = Anim Dynamics - Bone: dyn_armband_hi(ComponentPose=@n3, Alpha=0.5, ExternalForce="0,0,0") {
    alpha=0.9
    angularDampingOverride=0.98
    angularSpringConstant=1500
    bAngularSpring=true
    bGravityOverrideInSimSpace=true
    bOverrideAngularDamping=true
    numSolverIterationsPostUpdate=4
    numSolverIterationsPreUpdate=16
    simulationSpace=Actor
    alphaBoolBlend {blendOption=Linear}
    boundBone {boneName=dyn_armband_hi}
    physicsBodyDefinitions [{
      sphereCollisionRadius=0
      boundBone {boneName=dyn_armband_hi}
      boxExtents=<15 5 5>
      constraintSetup {angularTargetAxis=AxisZ, twistAxis=AxisY, angularLimitsMax=<15 15 15>, angularLimitsMin=<-15 -15 -15>, angularTarget=<0 0 1>}
      localJointOffset=<0 0 0>
     }]
   }
   n5 = Anim Dynamics - Bone: dyn_shldrA_leafA(ComponentPose=@n4, Alpha=0.45, ExternalForce="0,0,0") {
    angularDampingOverride=0.985
    angularSpringConstant=500
    bAngularSpring=true
    bGravityOverrideInSimSpace=true
    bLinearSpring=true
    bOverrideAngularDamping=true
    bOverrideLinearDamping=true
    gravityScale=0.085
    linearDampingOverride=0.99
    linearSpringConstant=600
    alphaBoolBlend {blendOption=Linear}
    boundBone {boneName=dyn_shldrA_leafA}
    componentAppliedLinearAccClamp=<1000 1000 1000>
    componentLinearVelScale=<100 100 100>
    physicsBodyDefinitions [{
      sphereCollisionRadius=0
      boundBone {boneName=dyn_shldrA_leafA}
      boxExtents=<30 30 30>
      constraintSetup {
       angularTargetAxis=AxisY
       angularLimitsMax=<0 0 0>
       angularLimitsMin=<0 0 0>
       angularTarget=<0 0 0>
       linearAxesMax=<3 2 0.5>
       linearAxesMin=<-0.5 0 -0.5>
      }
      localJointOffset=<0 0 0>
     }]
   }
   n6 = Anim Dynamics - Bone: dyn_shldrA_leafB(ComponentPose=@n5, Alpha=0.3, ExternalForce="0,0,0") {
    angularDampingOverride=0.985
    angularSpringConstant=300
    bAngularSpring=true
    bGravityOverrideInSimSpace=true
    bLinearSpring=true
    bOverrideAngularDamping=true
    bOverrideLinearDamping=true
    gravityScale=0.15
    linearDampingOverride=0.99
    linearSpringConstant=200
    alphaBoolBlend {blendOption=Linear}
    boundBone {boneName=dyn_shldrA_leafB}
    componentAppliedLinearAccClamp=<1000 1000 1000>
    componentLinearVelScale=<100 100 100>
    physicsBodyDefinitions [{
      sphereCollisionRadius=0
      boundBone {boneName=dyn_shldrA_leafB}
      boxExtents=<30 45 7>
      constraintSetup {
       angularTargetAxis=AxisY
       angularLimitsMax=<0 0 0>
       angularLimitsMin=<0 0 0>
       angularTarget=<0 0 0>
       linearAxesMax=<3 2 0.5>
       linearAxesMin=<-0.5 0 -0.5>
      }
      localJointOffset=<0 0 0>
     }]
   }
   n7 = Anim Dynamics - Bone: dyn_shldrA_leafC(ComponentPose=@n6, Alpha=0.35, ExternalForce="0,0,0") {
    angularDampingOverride=0.985
    angularSpringConstant=500
    bAngularSpring=true
    bGravityOverrideInSimSpace=true
    bLinearSpring=true
    bOverrideAngularDamping=true
    bOverrideLinearDamping=true
    gravityScale=0.3
    linearDampingOverride=0.99
    linearSpringConstant=500
    alphaBoolBlend {blendOption=Linear}
    boundBone {boneName=dyn_shldrA_leafC}
    componentAppliedLinearAccClamp=<1000 1000 1000>
    componentLinearVelScale=<100 100 100>
    physicsBodyDefinitions [{
      sphereCollisionRadius=0
      boundBone {boneName=dyn_shldrA_leafC}
      boxExtents=<20 10 15>
      constraintSetup {
       angularTargetAxis=AxisY
       angularLimitsMax=<0 0 0>
       angularLimitsMin=<0 0 0>
       angularTarget=<0 0 0>
       linearAxesMax=<3 2 0.5>
       linearAxesMin=<-0.5 0 -0.5>
      }
      localJointOffset=<0 0 0>
     }]
   }
   n8 = Anim Dynamics - Bone: dyn_shldrB_leafC(ComponentPose=@n7, Alpha=0.45, ExternalForce="0,0,0") {
    angularDampingOverride=0.985
    angularSpringConstant=500
    bAngularSpring=true
    bGravityOverrideInSimSpace=true
    bLinearSpring=true
    bOverrideAngularDamping=true
    bOverrideLinearDamping=true
    gravityScale=0.085
    linearDampingOverride=0.99
    linearSpringConstant=600
    alphaBoolBlend {blendOption=Linear}
    boundBone {boneName=dyn_shldrB_leafC}
    componentAppliedLinearAccClamp=<1000 1000 1000>
    componentLinearVelScale=<100 100 100>
    physicsBodyDefinitions [{
      sphereCollisionRadius=0
      boundBone {boneName=dyn_shldrB_leafC}
      boxExtents=<30 30 30>
      constraintSetup {
       angularTargetAxis=AxisY
       angularLimitsMax=<0 0 0>
       angularLimitsMin=<0 0 0>
       angularTarget=<0 0 0>
       linearAxesMax=<3 2 0.5>
       linearAxesMin=<-0.5 0 -0.5>
      }
      localJointOffset=<0 0 0>
     }]
   }
   n9 = Anim Dynamics - Bone: dyn_shldrB_leafB(ComponentPose=@n8, Alpha=0.3, ExternalForce="0,0,0") {
    angularDampingOverride=0.985
    angularSpringConstant=300
    bAngularSpring=true
    bGravityOverrideInSimSpace=true
    bLinearSpring=true
    bOverrideAngularDamping=true
    bOverrideLinearDamping=true
    gravityScale=0.15
    linearDampingOverride=0.99
    linearSpringConstant=200
    alphaBoolBlend {blendOption=Linear}
    boundBone {boneName=dyn_shldrB_leafB}
    componentAppliedLinearAccClamp=<1000 1000 1000>
    componentLinearVelScale=<100 100 100>
    physicsBodyDefinitions [{
      sphereCollisionRadius=0
      boundBone {boneName=dyn_shldrB_leafB}
      boxExtents=<30 45 7>
      constraintSetup {
       angularTargetAxis=AxisY
       angularLimitsMax=<0 0 0>
       angularLimitsMin=<0 0 0>
       angularTarget=<0 0 0>
       linearAxesMax=<3 2 0.5>
       linearAxesMin=<-0.5 0 -0.5>
      }
      localJointOffset=<0 0 0>
     }]
   }
   n10 = Anim Dynamics - Bone: dyn_shldrB_leafA(ComponentPose=@n9, Alpha=0.35, ExternalForce="0,0,0") {
    angularDampingOverride=0.985
    angularSpringConstant=500
    bAngularSpring=true
    bGravityOverrideInSimSpace=true
    bLinearSpring=true
    bOverrideAngularDamping=true
    bOverrideLinearDamping=true
    gravityScale=0.3
    linearDampingOverride=0.99
    linearSpringConstant=500
    alphaBoolBlend {blendOption=Linear}
    boundBone {boneName=dyn_shldrB_leafA}
    componentAppliedLinearAccClamp=<1000 1000 1000>
    componentLinearVelScale=<100 100 100>
    physicsBodyDefinitions [{
      sphereCollisionRadius=0
      boundBone {boneName=dyn_shldrB_leafA}
      boxExtents=<20 10 15>
      constraintSetup {
       angularTargetAxis=AxisY
       angularLimitsMax=<0 0 0>
       angularLimitsMin=<0 0 0>
       angularTarget=<0 0 0>
       linearAxesMax=<3 2 0.5>
       linearAxesMin=<-0.5 0 -0.5>
      }
      localJointOffset=<0 0 0>
     }]
   }
   n11 = Anim Dynamics - Bone: dyn_shldr_bck_leafB(ComponentPose=@n10, Alpha=0.45, ExternalForce="0,0,0") {
    angularDampingOverride=0.985
    angularSpringConstant=500
    bAngularSpring=true
    bGravityOverrideInSimSpace=true
    bLinearSpring=true
    bOverrideAngularDamping=true
    bOverrideLinearDamping=true
    gravityScale=0.085
    linearDampingOverride=0.99
    linearSpringConstant=600
    alphaBoolBlend {blendOption=Linear}
    boundBone {boneName=dyn_shldr_bck_leafB}
    componentAppliedLinearAccClamp=<1000 1000 1000>
    componentLinearVelScale=<100 100 100>
    physicsBodyDefinitions [{
      sphereCollisionRadius=0
      boundBone {boneName=dyn_shldr_bck_leafB}
      boxExtents=<30 30 30>
      constraintSetup {
       angularTargetAxis=AxisY
       angularLimitsMax=<0 0 0>
       angularLimitsMin=<0 0 0>
       angularTarget=<0 0 0>
       linearAxesMax=<3 2 0.5>
       linearAxesMin=<-0.5 0 -0.5>
      }
      localJointOffset=<0 0 0>
     }]
   }
   n12 = Anim Dynamics - Bone: dyn_shldr_bck_leafA(ComponentPose=@n11, Alpha=0.3, ExternalForce="0,0,0") {
    angularDampingOverride=0.985
    angularSpringConstant=300
    bAngularSpring=true
    bGravityOverrideInSimSpace=true
    bLinearSpring=true
    bOverrideAngularDamping=true
    bOverrideLinearDamping=true
    gravityScale=0.15
    linearDampingOverride=0.99
    linearSpringConstant=200
    alphaBoolBlend {blendOption=Linear}
    boundBone {boneName=dyn_shldr_bck_leafA}
    componentAppliedLinearAccClamp=<1000 1000 1000>
    componentLinearVelScale=<100 100 100>
    physicsBodyDefinitions [{
      sphereCollisionRadius=0
      boundBone {boneName=dyn_shldr_bck_leafA}
      boxExtents=<30 45 7>
      constraintSetup {
       angularTargetAxis=AxisY
       angularLimitsMax=<0 0 0>
       angularLimitsMin=<0 0 0>
       angularTarget=<0 0 0>
       linearAxesMax=<3 2 0.5>
       linearAxesMin=<-0.5 0 -0.5>
      }
      localJointOffset=<0 0 0>
     }]
   }
   n13 = Anim Dynamics - Bone: dyn_shldr_bck_leafC(ComponentPose=@n12, Alpha=0.35, ExternalForce="0,0,0") {
    angularDampingOverride=0.985
    angularSpringConstant=500
    bAngularSpring=true
    bGravityOverrideInSimSpace=true
    bLinearSpring=true
    bOverrideAngularDamping=true
    bOverrideLinearDamping=true
    gravityScale=0.3
    linearDampingOverride=0.99
    linearSpringConstant=500
    alphaBoolBlend {blendOption=Linear}
    boundBone {boneName=dyn_shldr_bck_leafC}
    componentAppliedLinearAccClamp=<1000 1000 1000>
    componentLinearVelScale=<100 100 100>
    physicsBodyDefinitions [{
      sphereCollisionRadius=0
      boundBone {boneName=dyn_shldr_bck_leafC}
      boxExtents=<20 10 15>
      constraintSetup {
       angularTargetAxis=AxisY
       angularLimitsMax=<0 0 0>
       angularLimitsMin=<0 0 0>
       angularTarget=<0 0 0>
       linearAxesMax=<3 2 0.5>
       linearAxesMin=<-0.5 0 -0.5>
      }
      localJointOffset=<0 0 0>
     }]
   }
   n14 = Anim Dynamics - Bone: dyn_pouchB_leafA(ComponentPose=@n13, Alpha=0.45, ExternalForce="0,0,0") {
    angularDampingOverride=0.985
    angularSpringConstant=500
    bAngularSpring=true
    bGravityOverrideInSimSpace=true
    bLinearSpring=true
    bOverrideAngularDamping=true
    bOverrideLinearDamping=true
    gravityScale=0.085
    linearDampingOverride=0.99
    linearSpringConstant=600
    alphaBoolBlend {blendOption=Linear}
    boundBone {boneName=dyn_pouchB_leafA}
    componentAppliedLinearAccClamp=<1000 1000 1000>
    componentLinearVelScale=<100 100 100>
    physicsBodyDefinitions [{
      sphereCollisionRadius=0
      boundBone {boneName=dyn_pouchB_leafA}
      boxExtents=<30 30 30>
      constraintSetup {
       angularTargetAxis=AxisY
       angularLimitsMax=<0 0 0>
       angularLimitsMin=<0 0 0>
       angularTarget=<0 0 0>
       linearAxesMax=<3 2 0.5>
       linearAxesMin=<-0.5 0 -0.5>
      }
      localJointOffset=<0 0 0>
     }]
   }
   n15 = Anim Dynamics - Bone: dyn_pouchB_leafB(ComponentPose=@n14, Alpha=0.3, ExternalForce="0,0,0") {
    angularDampingOverride=0.985
    angularSpringConstant=300
    bAngularSpring=true
    bGravityOverrideInSimSpace=true
    bLinearSpring=true
    bOverrideAngularDamping=true
    bOverrideLinearDamping=true
    gravityScale=0.15
    linearDampingOverride=0.99
    linearSpringConstant=200
    alphaBoolBlend {blendOption=Linear}
    boundBone {boneName=dyn_pouchB_leafB}
    componentAppliedLinearAccClamp=<1000 1000 1000>
    componentLinearVelScale=<100 100 100>
    physicsBodyDefinitions [{
      sphereCollisionRadius=0
      boundBone {boneName=dyn_pouchB_leafB}
      boxExtents=<30 45 7>
      constraintSetup {
       angularTargetAxis=AxisY
       angularLimitsMax=<0 0 0>
       angularLimitsMin=<0 0 0>
       angularTarget=<0 0 0>
       linearAxesMax=<3 2 0.5>
       linearAxesMin=<-0.5 0 -0.5>
      }
      localJointOffset=<0 0 0>
     }]
   }
   n16 = Anim Dynamics - Bone: dyn_pouchB_leafC(ComponentPose=@n15, Alpha=0.35, ExternalForce="0,0,0") {
    angularDampingOverride=0.985
    angularSpringConstant=500
    bAngularSpring=true
    bGravityOverrideInSimSpace=true
    bLinearSpring=true
    bOverrideAngularDamping=true
    bOverrideLinearDamping=true
    gravityScale=0.3
    linearDampingOverride=0.99
    linearSpringConstant=500
    alphaBoolBlend {blendOption=Linear}
    boundBone {boneName=dyn_pouchB_leafC}
    componentAppliedLinearAccClamp=<1000 1000 1000>
    componentLinearVelScale=<100 100 100>
    physicsBodyDefinitions [{
      sphereCollisionRadius=0
      boundBone {boneName=dyn_pouchB_leafC}
      boxExtents=<20 10 15>
      constraintSetup {
       angularTargetAxis=AxisY
       angularLimitsMax=<0 0 0>
       angularLimitsMin=<0 0 0>
       angularTarget=<0 0 0>
       linearAxesMax=<3 2 0.5>
       linearAxesMin=<-0.5 0 -0.5>
      }
      localJointOffset=<0 0 0>
     }]
   }
   n17 = Anim Dynamics - Bone: dyn_pouch_leafC(ComponentPose=@n16, Alpha=0.45, ExternalForce="0,0,0") {
    angularDampingOverride=0.985
    angularSpringConstant=500
    bAngularSpring=true
    bGravityOverrideInSimSpace=true
    bLinearSpring=true
    bOverrideAngularDamping=true
    bOverrideLinearDamping=true
    gravityScale=0.085
    linearDampingOverride=0.99
    linearSpringConstant=600
    alphaBoolBlend {blendOption=Linear}
    boundBone {boneName=dyn_pouch_leafC}
    componentAppliedLinearAccClamp=<1000 1000 1000>
    componentLinearVelScale=<100 100 100>
    physicsBodyDefinitions [{
      sphereCollisionRadius=0
      boundBone {boneName=dyn_pouch_leafC}
      boxExtents=<30 30 30>
      constraintSetup {
       angularTargetAxis=AxisY
       angularLimitsMax=<0 0 0>
       angularLimitsMin=<0 0 0>
       angularTarget=<0 0 0>
       linearAxesMax=<3 2 0.5>
       linearAxesMin=<-0.5 0 -0.5>
      }
      localJointOffset=<0 0 0>
     }]
   }
   n18 = Anim Dynamics - Bone: dyn_pouch_leafA(ComponentPose=@n17, Alpha=0.3, ExternalForce="0,0,0") {
    angularDampingOverride=0.985
    angularSpringConstant=300
    bAngularSpring=true
    bGravityOverrideInSimSpace=true
    bLinearSpring=true
    bOverrideAngularDamping=true
    bOverrideLinearDamping=true
    gravityScale=0.15
    linearDampingOverride=0.99
    linearSpringConstant=200
    alphaBoolBlend {blendOption=Linear}
    boundBone {boneName=dyn_pouch_leafA}
    componentAppliedLinearAccClamp=<1000 1000 1000>
    componentLinearVelScale=<100 100 100>
    physicsBodyDefinitions [{
      sphereCollisionRadius=0
      boundBone {boneName=dyn_pouch_leafA}
      boxExtents=<30 45 7>
      constraintSetup {
       angularTargetAxis=AxisY
       angularLimitsMax=<0 0 0>
       angularLimitsMin=<0 0 0>
       angularTarget=<0 0 0>
       linearAxesMax=<3 2 0.5>
       linearAxesMin=<-0.5 0 -0.5>
      }
      localJointOffset=<0 0 0>
     }]
   }
   n19 = Anim Dynamics - Bone: dyn_pouch_leafB(ComponentPose=@n18, Alpha=0.35, ExternalForce="0,0,0") {
    angularDampingOverride=0.985
    angularSpringConstant=500
    bAngularSpring=true
    bGravityOverrideInSimSpace=true
    bLinearSpring=true
    bOverrideAngularDamping=true
    bOverrideLinearDamping=true
    gravityScale=0.3
    linearDampingOverride=0.99
    linearSpringConstant=500
    alphaBoolBlend {blendOption=Linear}
    boundBone {boneName=dyn_pouch_leafB}
    componentAppliedLinearAccClamp=<1000 1000 1000>
    componentLinearVelScale=<100 100 100>
    physicsBodyDefinitions [{
      sphereCollisionRadius=0
      boundBone {boneName=dyn_pouch_leafB}
      boxExtents=<20 10 15>
      constraintSetup {
       angularTargetAxis=AxisY
       angularLimitsMax=<0 0 0>
       angularLimitsMin=<0 0 0>
       angularTarget=<0 0 0>
       linearAxesMax=<3 2 0.5>
       linearAxesMin=<-0.5 0 -0.5>
      }
      localJointOffset=<0 0 0>
     }]
   }
   n20 = Anim Dynamics - Bone: dyn_boot_leafC(ComponentPose=@n19, Alpha=0.45, ExternalForce="0,0,0") {
    angularDampingOverride=0.985
    angularSpringConstant=500
    bAngularSpring=true
    bGravityOverrideInSimSpace=true
    bLinearSpring=true
    bOverrideAngularDamping=true
    bOverrideLinearDamping=true
    gravityScale=0.085
    linearDampingOverride=0.99
    linearSpringConstant=600
    alphaBoolBlend {blendOption=Linear}
    boundBone {boneName=dyn_boot_leafC}
    componentAppliedLinearAccClamp=<1000 1000 1000>
    componentLinearVelScale=<100 100 100>
    physicsBodyDefinitions [{
      sphereCollisionRadius=0
      boundBone {boneName=dyn_boot_leafC}
      boxExtents=<30 30 30>
      constraintSetup {
       angularTargetAxis=AxisY
       angularLimitsMax=<0 0 0>
       angularLimitsMin=<0 0 0>
       angularTarget=<0 0 0>
       linearAxesMax=<3 2 0.5>
       linearAxesMin=<-0.5 0 -0.5>
      }
      localJointOffset=<0 0 0>
     }]
   }
   n21 = Anim Dynamics - Bone: dyn_boot_leafB(ComponentPose=@n20, Alpha=0.3, ExternalForce="0,0,0") {
    angularDampingOverride=0.985
    angularSpringConstant=300
    bAngularSpring=true
    bGravityOverrideInSimSpace=true
    bLinearSpring=true
    bOverrideAngularDamping=true
    bOverrideLinearDamping=true
    gravityScale=0.15
    linearDampingOverride=0.99
    linearSpringConstant=200
    alphaBoolBlend {blendOption=Linear}
    boundBone {boneName=dyn_boot_leafB}
    componentAppliedLinearAccClamp=<1000 1000 1000>
    componentLinearVelScale=<100 100 100>
    physicsBodyDefinitions [{
      sphereCollisionRadius=0
      boundBone {boneName=dyn_boot_leafB}
      boxExtents=<30 45 7>
      constraintSetup {
       angularTargetAxis=AxisY
       angularLimitsMax=<0 0 0>
       angularLimitsMin=<0 0 0>
       angularTarget=<0 0 0>
       linearAxesMax=<3 2 0.5>
       linearAxesMin=<-0.5 0 -0.5>
      }
      localJointOffset=<0 0 0>
     }]
   }
   n22 = Anim Dynamics - Bone: dyn_boot_leafA(ComponentPose=@n21, Alpha=0.35, ExternalForce="0,0,0") {
    angularDampingOverride=0.985
    angularSpringConstant=500
    bAngularSpring=true
    bGravityOverrideInSimSpace=true
    bLinearSpring=true
    bOverrideAngularDamping=true
    bOverrideLinearDamping=true
    gravityScale=0.3
    linearDampingOverride=0.99
    linearSpringConstant=500
    alphaBoolBlend {blendOption=Linear}
    boundBone {boneName=dyn_boot_leafA}
    componentAppliedLinearAccClamp=<1000 1000 1000>
    componentLinearVelScale=<100 100 100>
    physicsBodyDefinitions [{
      sphereCollisionRadius=0
      boundBone {boneName=dyn_boot_leafA}
      boxExtents=<20 10 15>
      constraintSetup {
       angularTargetAxis=AxisY
       angularLimitsMax=<0 0 0>
       angularLimitsMin=<0 0 0>
       angularTarget=<0 0 0>
       linearAxesMax=<3 2 0.5>
       linearAxesMin=<-0.5 0 -0.5>
      }
      localJointOffset=<0 0 0>
     }]
   }
   n23 = Anim Dynamics - Bone: dyn_tongue_02(ComponentPose=@n22, Alpha=0.3, ExternalForce="0,0,0") {
    angularDampingOverride=0.985
    angularSpringConstant=100
    bAngularSpring=true
    bGravityOverrideInSimSpace=true
    bLinearSpring=true
    bOverrideAngularDamping=true
    bOverrideLinearDamping=true
    gravityScale=0.3
    linearDampingOverride=0.99
    linearSpringConstant=100
    alphaBoolBlend {blendOption=Linear}
    boundBone {boneName=dyn_tongue_02}
    componentAppliedLinearAccClamp=<1000 1000 1000>
    componentLinearVelScale=<100 100 100>
    physicsBodyDefinitions [{
      sphereCollisionRadius=0
      boundBone {boneName=dyn_tongue_02}
      boxExtents=<5 15 15>
      constraintSetup {
       angularTargetAxis=AxisY
       angularLimitsMax=<0 0 0>
       angularLimitsMin=<0 0 0>
       angularTarget=<0 0 0>
       linearAxesMax=<2 6 3>
       linearAxesMin=<-2 -2 -3>
      }
      localJointOffset=<0 0 0>
     }]
    relativeSpaceBone {boneName=dyn_tongue_01}
   }
   n24 = Anim Dynamics - Bone: dyn_boot_fork(ComponentPose=@n23, Alpha=0.45, ExternalForce="0,0,0") {
    angularDampingOverride=0.985
    angularSpringConstant=200
    bAngularSpring=true
    bGravityOverrideInSimSpace=true
    bOverrideAngularDamping=true
    bOverrideLinearDamping=true
    gravityScale=0.35
    linearDampingOverride=0.99
    linearSpringConstant=200
    simulationSpace=RootRelative
    alphaBoolBlend {blendOption=Linear}
    boundBone {boneName=dyn_boot_fork}
    componentAppliedLinearAccClamp=<1000 1000 1000>
    componentLinearVelScale=<100 100 100>
    physicsBodyDefinitions [{
      sphereCollisionRadius=0
      boundBone {boneName=dyn_boot_fork}
      boxExtents=<30 30 30>
      constraintSetup {angularTargetAxis=AxisY, angularLimitsMax=<15 1 20>, angularLimitsMin=<-15 -1 -20>, angularTarget=<0 0 0>}
      localJointOffset=<0 7 0>
     }]
   }
   n25 = Anim Dynamics - Bone: dyn_pouch_strand_01(ComponentPose=@n24, Alpha=0.45, ExternalForce="0,0,0") {
    angularDampingOverride=0.985
    angularSpringConstant=200
    bAngularSpring=true
    bGravityOverrideInSimSpace=true
    bLinearSpring=true
    bOverrideAngularDamping=true
    bOverrideLinearDamping=true
    gravityScale=0.275
    linearDampingOverride=0.99
    linearSpringConstant=200
    simulationSpace=RootRelative
    alphaBoolBlend {blendOption=Linear}
    boundBone {boneName=dyn_pouch_strand_01}
    componentAppliedLinearAccClamp=<1000 1000 1000>
    componentLinearVelScale=<100 100 100>
    physicsBodyDefinitions [{
      sphereCollisionRadius=0
      boundBone {boneName=dyn_pouch_strand_01}
      boxExtents=<30 30 30>
      constraintSetup {angularTargetAxis=AxisY, angularLimitsMax=<0 0 20>, angularLimitsMin=<0 -20 -20>, angularTarget=<0 0 0>}
      localJointOffset=<10 0 0>
     }]
    relativeSpaceBone {boneName=pelvis}
   }
   n26 = Anim Dynamics - Bone: dyn_pouchB_strandB_01(ComponentPose=@n25, Alpha=0.55, ExternalForce="0,0,0") {
    angularDampingOverride=0.985
    angularSpringConstant=200
    bAngularSpring=true
    bGravityOverrideInSimSpace=true
    bLinearSpring=true
    bOverrideAngularDamping=true
    bOverrideLinearDamping=true
    gravityScale=0.275
    linearDampingOverride=0.99
    linearSpringConstant=200
    simulationSpace=RootRelative
    alphaBoolBlend {blendOption=Linear}
    boundBone {boneName=dyn_pouchB_strandB_01}
    componentAppliedLinearAccClamp=<1000 1000 1000>
    componentLinearVelScale=<100 100 100>
    physicsBodyDefinitions [{
      sphereCollisionRadius=0
      boundBone {boneName=dyn_pouchB_strandB_01}
      boxExtents=<30 30 30>
      constraintSetup {angularTargetAxis=AxisY, angularLimitsMax=<0 20 20>, angularLimitsMin=<0 -20 -20>, angularTarget=<0 0 0>}
      localJointOffset=<10 0 0>
     }]
    relativeSpaceBone {boneName=dyn_pouchB_leaf}
   }
   n27 = Anim Dynamics - Bone: dyn_pouch_strand_02(ComponentPose=@n26, Alpha=0.45, ExternalForce="0,0,0") {
    angularDampingOverride=0.985
    angularSpringConstant=500
    bAngularSpring=true
    bGravityOverrideInSimSpace=true
    bLinearSpring=true
    bOverrideAngularDamping=true
    bOverrideLinearDamping=true
    gravityScale=0.25
    linearDampingOverride=0.99
    linearSpringConstant=600
    alphaBoolBlend {blendOption=Linear}
    boundBone {boneName=dyn_pouch_strand_02}
    componentAppliedLinearAccClamp=<1000 1000 1000>
    componentLinearVelScale=<100 100 100>
    physicsBodyDefinitions [{
      sphereCollisionRadius=0
      boundBone {boneName=dyn_pouch_strand_02}
      boxExtents=<30 30 30>
      constraintSetup {
       angularTargetAxis=AxisY
       angularLimitsMax=<0 0 0>
       angularLimitsMin=<0 0 0>
       angularTarget=<0 0 0>
       linearAxesMax=<3 2 0.5>
       linearAxesMin=<-0.5 0 -0.5>
      }
      localJointOffset=<0 0 0>
     }]
   }
   n28 = Anim Dynamics - Bone: dyn_pouchB_strandB_02(ComponentPose=@n27, Alpha=0.52, ExternalForce="0,0,0") {
    angularDampingOverride=0.985
    angularSpringConstant=500
    bAngularSpring=true
    bGravityOverrideInSimSpace=true
    bLinearSpring=true
    bOverrideAngularDamping=true
    bOverrideLinearDamping=true
    gravityScale=0.3
    linearDampingOverride=0.99
    linearSpringConstant=600
    alphaBoolBlend {blendOption=Linear}
    boundBone {boneName=dyn_pouchB_strandB_02}
    componentAppliedLinearAccClamp=<1000 1000 1000>
    componentLinearVelScale=<100 100 100>
    physicsBodyDefinitions [{
      sphereCollisionRadius=0
      boundBone {boneName=dyn_pouchB_strandB_02}
      boxExtents=<30 30 30>
      constraintSetup {
       angularTargetAxis=AxisY
       angularLimitsMax=<0 0 0>
       angularLimitsMin=<0 0 0>
       angularTarget=<0 0 0>
       linearAxesMax=<3 2 0.5>
       linearAxesMin=<-0.5 0 -0.5>
      }
      localJointOffset=<0 0 0>
     }]
   }
   n29 = "Transform (Modify) Bone - Bone: dyn_pouch_leaf"(ComponentPose=@n28, Alpha=Get pelvis_out() -> (pelvis_out:double), Translation="0,1.4,-1.85", Rotation="0,40,0", Scale="1,1,1") {
    rotationMode=BMM_Additive
    translationMode=BMM_Additive
    alphaBoolBlend {blendOption=Linear}
    alphaScaleBiasClamp {bInterpResult=true}
    boneToModify {boneName=dyn_pouch_leaf}
    rotation {yaw=40}
    translation {y=1.4, z=-1.85}
   }
   n30 = Anim Dynamics - Bone: dyn_shldr_vineA(ComponentPose=@n29, Alpha=0.45, ExternalForce="0,0,0") {
    angularDampingOverride=0.985
    angularSpringConstant=100
    bAngularSpring=true
    bGravityOverrideInSimSpace=true
    bLinearSpring=true
    bOverrideAngularDamping=true
    bOverrideLinearDamping=true
    gravityScale=0.3
    linearDampingOverride=0.99
    linearSpringConstant=100
    alphaBoolBlend {blendOption=Linear}
    boundBone {boneName=dyn_shldr_vineA}
    componentAppliedLinearAccClamp=<500 500 500>
    componentLinearVelScale=<200 200 200>
    physicsBodyDefinitions [{
      sphereCollisionRadius=0
      boundBone {boneName=dyn_shldr_vineA}
      boxExtents=<30 30 30>
      constraintSetup {
       angularTargetAxis=AxisY
       angularLimitsMax=<0 0 0>
       angularLimitsMin=<0 0 0>
       angularTarget=<0 0 0>
       linearAxesMax=<0.1 1 1.8>
       linearAxesMin=<-3 -1 -1.2>
      }
      localJointOffset=<0 0 0>
     }]
   }
   n31 = Anim Dynamics - Bone: dyn_shldr_vineB(ComponentPose=@n30, Alpha=0.1, ExternalForce="0,0,0") {
    angularDampingOverride=0.985
    angularSpringConstant=90
    bAngularSpring=true
    bGravityOverrideInSimSpace=true
    bLinearSpring=true
    bOverrideAngularDamping=true
    bOverrideLinearDamping=true
    gravityScale=0.375
    linearDampingOverride=0.99
    linearSpringConstant=90
    alphaBoolBlend {blendOption=Linear}
    boundBone {boneName=dyn_shldr_vineB}
    componentAppliedLinearAccClamp=<500 500 500>
    componentLinearVelScale=<200 200 200>
    physicsBodyDefinitions [{
      sphereCollisionRadius=0
      boundBone {boneName=dyn_shldr_vineB}
      boxExtents=<30 30 30>
      constraintSetup {
       angularTargetAxis=AxisY
       angularLimitsMax=<0 0 0>
       angularLimitsMin=<0 0 0>
       angularTarget=<0 0 0>
       linearAxesMax=<3 2 1.8>
       linearAxesMin=<-2 -1.5 -1.2>
      }
      localJointOffset=<0 0 0>
     }]
   }
   n32 = Flap Node - Bone: scarf_back!(ComponentPose:Component Space Pose Link Structure=@n31, Alpha:float=1) -> (Pose:Component Space Pose Link Structure) {
    is=title FortAnimGraphNode_Flap
    bUseClamp=true
    lODThreshold=2
    targetClampMax=3.14
    targetComponent=RotationY
    alphaBoolBlend {blendOption=Linear}
    sourceDrives [{
      bUseRange=false
      drivingCurve=null
      multiplier=1
      rangeMax=1
      rangeMin=-1
      remappedMax=1
      remappedMin=0
      sourceComponent=RotationZ
      useQuaternion=true
      sourceBone {boneName=head}
     }]
    targetBone {boneName=scarf_back}
   }
   n33 = Flap Node - Bone: scarf_back(ComponentPose=@n32, Alpha=1) {
    bUseClamp=true
    lODThreshold=2
    targetClampMax=100
    targetClampMin=-100
    targetComponent=RotationZ
    alphaBoolBlend {blendOption=Linear}
    sourceDrives [{
      bUseRange=false
      drivingCurve=null
      multiplier=-0.2
      rangeMax=1
      rangeMin=-1
      remappedMax=1
      remappedMin=0
      sourceComponent=RotationX
      useQuaternion=true
      sourceBone {boneName=head}
     }]
    targetBone {boneName=scarf_back}
   }
   n34 = Flap Node - Bone: scarf_back(ComponentPose=@n33, Alpha=1) {
    bUseClamp=true
    lODThreshold=2
    targetClampMax=100
    targetClampMin=0.05
    alphaBoolBlend {blendOption=Linear}
    sourceDrives [{
      bUseRange=false
      drivingCurve=null
      multiplier=-0.55
      rangeMax=1
      rangeMin=-1
      remappedMax=1
      remappedMin=0
      sourceComponent=RotationX
      useQuaternion=true
      sourceBone {boneName=head}
     }]
    targetBone {boneName=scarf_back}
   }
   n35 = Flap Node - Bone: scarf_back(ComponentPose=@n34, Alpha=1) {
    bUseClamp=true
    lODThreshold=2
    targetClampMax=-0.05
    targetClampMin=-0.2
    targetComponent=ScaleX
    alphaBoolBlend {blendOption=Linear}
    sourceDrives [{
      bUseRange=false
      drivingCurve=null
      multiplier=-0.25
      rangeMax=1
      rangeMin=-1
      remappedMax=1
      remappedMin=0
      sourceComponent=RotationX
      useQuaternion=true
      sourceBone {boneName=head}
     }]
    targetBone {boneName=scarf_back}
   }
   Output Pose(Result=Component To Local(ComponentPose=@n35)) {layerGroup=None}
   AnimGraphNode_SubInput_0 = Input Pose()
  ]}
 EventGraph {nodes [
   Event BlueprintUpdateAnimation()
   n1 = GetOwnerSkeletalMeshComponent()
   n2 = BreakRotator(InRot=BreakTransform(InTransform=GetDeltaTransformFromRefPose(self=@n1, BoneName=spine_05, BaseName=spine_01)).Rotation)
   n3 = MapRangeClamped(Value=@n2.Yaw, InRangeA=0, InRangeB=30)
   > Set pelvis_out(pelvis_out:double=@n3) -> (Output_Get:double)
   > Set pelvis_out(pelvis_out=float + float(A=MapRangeClamped(Value=@n2.Yaw, InRangeA=0, InRangeB=-20), B=Get pelvis_out()) {is=title K2Node_CommutativeAssociativeBinaryOperator})
   MakeRotator(Yaw=@n3)
  ]}
]
variables [pelvis_out {type=number}]
